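# Cloth-model anime-shaded material for the Nightingale backpack, with a periodically blinking emissive detail.
# A thresholded time sine gates a UV box mask, briefly swapping the PreFX emissive inside that region for the blink texture tinted by Blink Emissive Color.
# The result runs through MF_AnimeShading, then MF_AnimatedVisor, MF_HairPuffWPO and the shared character PostFX function, with a side branch into MF_ToonOutline_Mobile.
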Material /BRCosmetics/Accessories/FORT_Backpacks/Backpack_F_MED_Nightingale/Materials/M_Backpack_F_MED_Nightingale
graph {
 properties {
  bAutomaticallySetUsageInEditor=false
  bUseMaterialAttributes=true
  bUsedWithMeshDeformer=true
  bUsedWithSkeletalMesh=true
  pixelDepthOffsetMode=PDOM_Legacy
  shadingModel=MSM_Cloth
  usedShadingModels=Cloth
 }
 nodes [
  n1 = MaterialFunctionCall() {materialFunction=&BaseFN_CharacterFunction_PreFX}
  n2 = GetMaterialAttributes(None=@n1.Result) {attributeGetTypes [#1]}
  n3 = TextureSample() {
   samplerType=SAMPLERTYPE_Masks
   texture=/BRCosmetics/Accessories/FORT_Backpacks/Backpack_F_MED_Nightingale/Textures/T_F_MED_Nightingale_Backpack_E_Blink
  }
  n4 = Multiply(A=@n3.R, B=VectorParameter() {parameterName=Blink Emissive Color}.RGB)
  n5 = Add(A=Sine(None=Time()) {period=3}) {constB=-0.93}
  n6 = Multiply(
   A=Saturate(None=Sign(None=@n5))
   B=MaterialFunctionCall!(A=TextureCoordinate(), B=Constant2Vector() {g=0.38, r=0.475}, Bounds=Constant2Vector() {g=0.14, r=0.3}, Edge Falloff=Constant() {r=0}) {materialFunction=/Engine/Functions/Engine_MaterialFunctions01/Gradient/BoxMask-2D}.Result
  )
  n7 = MaterialFunctionCall!(
   Material=SetMaterialAttributes(MaterialAttributes=@n2.MaterialAttributes, Emissive Color=LinearInterpolate(A=@n2.Emissive Color, B=@n4, Alpha=@n6) {constB=0}) {attributeSetTypes [#1]}
   LitDiffuseOverride
   SSC_AlphaOverride
   Tint Mask
   DF InkLine Visibility Mask
   DF Inkline Injection
   DF Inkline Color Override
   DF Inkline Color Override Mask
   CustomUVs
  ) {materialFunction=/Game/Packages/Fortress_SharedMaterials/Anime_Master_Material/MF_AnimeShading}
  MaterialFunctionCall!(Material=@n7.Result) {materialFunction=/Game/Packages/Fortress_SharedMaterials/Anime_Master_Material/MF_ToonOutline_Mobile}
  MaterialFunctionCall(
   AO=@n1.AO
   Distance Blend=@n1.AO
   Fresnel=@n1.AO
   Gmap=@n1.AO
   Pre Skinned Local Position=@n1.AO
   Material=MaterialFunctionCall!(In=MaterialFunctionCall!(In=@n7.Result) {materialFunction=/Game/Packages/Fortress_SharedMaterials/Anime_Master_Material/MF_AnimatedVisor}.Result) {materialFunction=/Game/Characters/Player/Female/Medium/Bodies/F_MED_Peppermint/Materials/MF_HairPuffWPO}.Result
  ) {materialFunction=&BaseFN_CharacterFunction_PostFX}
 ]
}
